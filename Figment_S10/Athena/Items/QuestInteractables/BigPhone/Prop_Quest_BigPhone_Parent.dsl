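; Parent Blueprint for an indestructible giant rotary phone quest prop, where players dial numbers by damaging it.
; Hits on digit collision spheres map to numbers and drive timeline dial spins with audio.
; Each digit is checked against the quest sequence, so errors play a wrong-number sound while a full match credits tracked players, broadcasts OnPlayerDialed and rings.
; Clients watch quest updates for success FX; the paired-phone call handler only prints.
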(event EventBeginPlay
  ;[1 0.871 0.3434] Delay hack to wait until playlist data is loaded
  (Utilities|FlowControl|Delay 5.0)
  ;[0.4018 1 0.4524] Start listening to damage events
  (Development|PrintString "SetupPhone" (Variables|zDebug|GetbPrintDebug) (Variables|zDebug|GetbPrintDebug))
  (EventDispatchers|BindEventtoOnDamaged (EventDispatchers|CreateEvent :!event Custom|PhoneHit))
  ;[0.4018 1 0.4524] Set initial vars
  (Variables|DialFunctionality|SetDialStartingRotation (Variables|/Script/Engine.SceneComponent|GetRelativeRotation (Variables|Components|GetRotaryPhone_Dial)))
  (Variables|RingingFunctionality|SetInitialReceiverLocation (Variables|RingingFunctionality|GetInitialReceiverLocation))
  (Variables|RingingFunctionality|SetInitialReceiverRotation (Variables|/Script/Engine.SceneComponent|GetRelativeRotation (Variables|Components|GetRotaryPhone_Receiver)))
  ;[1 1 1 0.1] Quest and Initial Setup
  (bind _asfort_game_state_athena (Utilities|Casting|CastToFortGameStateAthena (Game|GetGameState)))
  (if (Variables|/Script/FortniteGame.FortGameStateAthena|GetbPlaylistDataIsLoaded _asfort_game_state_athena)
    (bind _cancompletequest (Class|AthenaQuestFunctionLibrary|AthenaCheckCreativeMode _asfort_game_state_athena))
    (Variables|yQuest-Backend&BP|SetbIsInPlayground (not _cancompletequest))
    ;;[0.4018 1 0.4524] Playgrounds Check
    (bind _gamestate (Class|FortAsyncActionCurrentPlaylistReady|PlaylistDataReadyAsync)
      (:OnReady
        (Development|PrintString "PlaylistDataReadyAsync" (Variables|zDebug|GetbPrintDebug) (Variables|zDebug|GetbPrintDebug))
        (bind _cancompletequest_2 (Class|AthenaQuestFunctionLibrary|AthenaCheckCreativeMode _gamestate))
        (Variables|yQuest-Backend&BP|SetbIsInPlayground (not _cancompletequest_2))
        ;;[0.4018 1 0.4524] Does player have quest, if so create and bind the event for when they complete it
        (Networking|SwitchHasAuthority
          (:Authority)
          (:Remote
            (bind (_questisvalid _questcompleted _objectivecompleted) (Default|AthenaCheckQuestandObjectiveCompletion (Utilities|Array|Get(acopy) (Fortnite|GetLocalFortPlayerControllers true)) (Variables|yQuest-Backend&BP|GetQuestItemDefinition) (Variables|yQuest-Backend&BP|GetQuestBackendObjectiveName)))
            (if (and _questisvalid (not _objectivecompleted))
              (Variables|yQuest-Backend&BP|SetbIsQuestLocallyValid true)
              (Development|PrintString "PhoneQuestValid")
              (Quests|BindEventtoOnQuestsUpdated (Quests|GetQuestManager (Utilities|Array|Get(acopy) (Fortnite|GetLocalFortPlayerControllers true)) "Invalid") (EventDispatchers|CreateEvent :!event Custom|ParentQuestUpdated))
              (else
                (Variables|yQuest-Backend&BP|SetbIsQuestLocallyValid false)))))))
    (else
      (Playlist|BindEventtoOnCurrentPlaylistLoaded _asfort_game_state_athena (delegate Custom|PlaylistLoaded)))))

(event Custom|PlaylistLoaded
  (params (PlaylistName :type Name) (PlaylistContextTags :type Gameplay%20Tag%20Container%20Structure%20%28by%20ref%29))
  ;[1 1 1 0.1] Quest and Initial Setup
  (bind _cancompletequest (Class|AthenaQuestFunctionLibrary|AthenaCheckCreativeMode (stale EventBeginPlay _asfort_game_state_athena)))
  (Variables|yQuest-Backend&BP|SetbIsInPlayground (not _cancompletequest))
  ;;[0.4018 1 0.4524] Playgrounds Check
  (bind _gamestate (Class|FortAsyncActionCurrentPlaylistReady|PlaylistDataReadyAsync)
    (:OnReady
      (Development|PrintString "PlaylistDataReadyAsync" (Variables|zDebug|GetbPrintDebug) (Variables|zDebug|GetbPrintDebug))
      (bind _cancompletequest_2 (Class|AthenaQuestFunctionLibrary|AthenaCheckCreativeMode _gamestate))
      (Variables|yQuest-Backend&BP|SetbIsInPlayground (not _cancompletequest_2))
      ;;[0.4018 1 0.4524] Does player have quest, if so create and bind the event for when they complete it
      (Networking|SwitchHasAuthority
        (:Authority)
        (:Remote
          (bind (_questisvalid _questcompleted _objectivecompleted) (Default|AthenaCheckQuestandObjectiveCompletion (Utilities|Array|Get(acopy) (Fortnite|GetLocalFortPlayerControllers true)) (Variables|yQuest-Backend&BP|GetQuestItemDefinition) (Variables|yQuest-Backend&BP|GetQuestBackendObjectiveName)))
          (if (and _questisvalid (not _objectivecompleted))
            (Variables|yQuest-Backend&BP|SetbIsQuestLocallyValid true)
            (Development|PrintString "PhoneQuestValid")
            (Quests|BindEventtoOnQuestsUpdated (Quests|GetQuestManager (Utilities|Array|Get(acopy) (Fortnite|GetLocalFortPlayerControllers true)) "Invalid") (EventDispatchers|CreateEvent :!event Custom|ParentQuestUpdated))
            (else
              (Variables|yQuest-Backend&BP|SetbIsQuestLocallyValid false))))))))

(event Custom|ServerBindDelegates
  ;[1 1 1 0.1] Quest and Initial Setup
  ;;[0.4018 1 0.4524] Is this needed?
  (Networking|SwitchHasAuthority
    (:Authority
      ;;[0.4018 1 0.4524] -Wait until match is ready
      ;; -Get other phone actors
      ;; -If they match our index, and are not ourselves
      ;; -Listen for their onPlayerDialed delegate
      (Class|FortAsyncActionAuthorityMatchReady|MatchReadyAsync
        (:OnReady
          (bind _outactors (Actor|GetAllActorsOfClass "/Figment_S10/Athena/Items/QuestInteractables/BigPhone/Prop_Quest_BigPhone_Parent.Prop_Quest_BigPhone_Parent_C"))
          (bind _output_get (Variables|RingingFunctionality|SetAllPhoneClassReferences _outactors))
          (for _array_element _output_get
            ;;[0.4018 1 0.4524] Find callers with my index, that aren't me
            (if (and (Utilities|NotEqual(Object) _array_element self) (== (Variables|RingingFunctionality|GetCallPairingInt) (Variables|RingingFunctionality|GetCallPairingInt _array_element)))
              (Default|BindEventtoOnPlayerDialed _array_element (EventDispatchers|CreateEvent :!event Custom|PlayerCalledMe)))))))))

(event Custom|PassByEvent_PlaySuccessFX
  (params (Controller :type Fort%20Player%20Controller%20Object%20Reference))
  ;[1 0.1087 0.7842] CUSTOM EVENT -> PlaySuccessFX
  (CallFunction|PlaySuccessFX Controller))

(event Custom|ParentQuestUpdated
  (bind _returnvalue (Fortnite|GetLocalFortPlayerControllers true))
  (bind _output (Utilities|Array|Get(acopy) _returnvalue))
  ;[1 1 1 0.1] Quest and Initial Setup
  ;;[0.4018 1 0.4524] Set IsLocallyValid = False, remove bound event
  ;;[1 0.1087 0.7842] EVENT -> ParentQuestUpdated
  (Development|PrintString "->ParentQuestUpdated" (Variables|zDebug|GetbPrintDebug) (Variables|zDebug|GetbPrintDebug))
  (if (Variables|yQuest-Backend&BP|GetbIsQuestLocallyValid)
    (bind (_questisvalid _questcompleted _objectivecompleted) (Default|AthenaCheckQuestandObjectiveCompletion (Utilities|Array|Get(acopy) (Fortnite|GetLocalFortPlayerControllers true)) (Variables|yQuest-Backend&BP|GetQuestItemDefinition) (Variables|yQuest-Backend&BP|GetQuestBackendObjectiveName)))
    (Development|PrintString "Parent Quest updated and we checked completion" (Variables|zDebug|GetbPrintDebug) (Variables|zDebug|GetbPrintDebug))
    (if (or _questisvalid _objectivecompleted)
      (Variables|yQuest-Backend&BP|SetbIsQuestLocallyValid false)
      (CallFunction|PassbyEventPlaySuccessFX _output)
      (Quests|UnbindEventfromOnQuestsUpdated (Quests|GetQuestManager _output "Invalid") (EventDispatchers|CreateEvent :!event Custom|ParentQuestUpdated)))))

(event Custom|PlayerCalledMe
  ;[1 0.1087 0.7842] EVENT -> OtherPhoneCalledMe
  (Development|PrintString "OtherPhoneCalledMe" (Variables|zDebug|GetbPrintDebug) (Variables|zDebug|GetbPrintDebug)))

(event Custom|PhoneHit
  (params (DamagedActor :type Actor%20Object%20Reference) (Damage :type Float%20%28single-precision%29) (InstigatedBy :type Controller%20Object%20Reference) (DamageCauser :type Actor%20Object%20Reference) (HitLocation :type Vector) (FHitComponent :type Primitive%20Component%20Object%20Reference) (BoneName :type Name) (Momentum :type Vector))
  (bind _dialcollisionspherearray (Variables|DialFunctionality|GetDialCollisionSphereArray))
  ;[0.4018 1 0.4524] Did player hit a number
  ;[0.4018 1 0.4524] Debug - what did we actually hit if not the collision
  (if (Utilities|Equal(Class) (Utilities|GetClass FHitComponent) "/Script/Engine.SphereComponent")
    ;[0.4018 1 0.4524] Are we ringing/rotating/dialing currently
    (if (Variables|DialFunctionality|GetbDialCanBeHit)
      ;[0.4018 1 0.4524] SET INITIAL VARIABLES
      (Utilities|FlowControl|Sequence
        (:then_0
          (Variables|DialFunctionality|SetDialDamageInstance Damage))
        (:then_1
          (Variables|DialFunctionality|SetDialPlayerWhoHit (Fortnite|GetFortPlayerControllerfromActor InstigatedBy)))
        (:then_2
          (Variables|DialFunctionality|SetDialHitLocation HitLocation))
        (:then_3
          (Variables|DialFunctionality|SetPhoneComponentHit FHitComponent)
          ;[0.4018 1 0.4524] We only get here if they hit a sphere anyway
          (bind _assphere_collision (Utilities|Casting|CastToSphereComponent (Variables|DialFunctionality|GetPhoneComponentHit))
            (:then
              ;[0.4018 1 0.4524] Set number we dialed and print debug string
              (Variables|DialFunctionality|SetDialNumberHit (Utilities|Map|Find _dialcollisionspherearray _assphere_collision))
              (Development|PrintString (Utilities|String|Append "HIT # = " (Utilities|String|ToString(Integer) (Utilities|Map|Find _dialcollisionspherearray _assphere_collision))) (Variables|zDebug|GetbPrintDebug) (Variables|zDebug|GetbPrintDebug))
              ;[1 0.1087 0.7842] -> RotateDial
              (CallFunction|RotateDial (Variables|DialFunctionality|GetPhoneComponentHit)))
            (:CastFailed
              (Development|PrintString "2. Didn't hit sphere collision component, or failed to cast" (Variables|zDebug|GetbPrintDebug) (Variables|zDebug|GetbPrintDebug)))))))
    (else
      (Development|PrintString (Utilities|String|Append "Phone component display name  = " (Utilities|GetDisplayName FHitComponent)) (Variables|zDebug|GetbPrintDebug) (Variables|zDebug|GetbPrintDebug)))))

(event Custom|RotateDial
  (params (NewParam :type Scene%20Component%20Object%20Reference))
  (bind _returnvalue 3e+01)
  (bind _returnvalue_2 2.5)
  (bind _numberhitforrotating (Variables|DialFunctionality|GetNumberHitForRotating))
  (bind _returnvalue_3 (* _numberhitforrotating 0.083))
  (bind _returnvalue_4 (- _returnvalue_2 _returnvalue_3))
  (bind _numberhitforrotating_2 (Variables|DialFunctionality|GetNumberHitForRotating))
  (bind _returnvalue_5 (* _numberhitforrotating_2 3e+01))
  (bind _returnvalue_6 (+ _returnvalue _returnvalue_5))
  (bind _rotary_initialspin_loop (Variables|Default|GetRotary_InitialSpin_Loop))
  (bind _rotary_returnspin_loop (Variables|Default|GetRotary_ReturnSpin_Loop))
  (bind _dialnumberhit (Variables|DialFunctionality|GetDialNumberHit))
  (Effects|SpawnEmitterAttached :EmitterTemplate "/Figment_S10/Weapons/FORT_Generic_Effects/Impacts/HuskFlesh/P_Impact_Electric_Husk_01.P_Impact_Electric_Husk_01" :AttachToComponent NewParam :AttachPointName "None" :Rotation (Variables|/Script/Engine.SceneComponent|GetRelativeRotation (Variables|Default|GetArrow)) :Scale "1.000000,1.000000,1.000000" :LocationType "KeepRelativeOffset" :bAutoDestroy true :PoolingMethod "None" :bAutoActivate true)
  ;[0.4018 1 0.4524] Block incoming input, this is checked earlier so it doesn't mess with any variables
  (Variables|DialFunctionality|SetbDialCanBeHit false)
  ;[0.4949 1 0.4643 0.1] Rotation Timelines/Math
  ;;[0.4018 1 0.4524] We need to rotate based on a number between 1 and 10.
  (if (== _dialnumberhit 0)
    (Variables|DialFunctionality|SetNumberHitForRotating 1e+01)
    (else
      (Variables|DialFunctionality|SetNumberHitForRotating (Math|Conversions|ToFloat(Integer) _dialnumberhit))))
  ;;[0.4018 1 0.4524] Play rate scaling from 1 to 2 (1-10)
  (Components|Timeline|SetPlayRate (Variables|Prop_Quest_BigPhone_Parent|GetTimelineDialRotation) _returnvalue_4)
  ;;[1 0.9092 0.2417] Audio - Rotary loop fade in
  (Components|Activation|Activate _rotary_initialspin_loop false)
  (Audio|Components|Audio|FadeIn _rotary_initialspin_loop 0.0 2.0 0.0 "Linear" :!class /Script/Engine.AudioComponent)
  ;;[0.4018 1 0.4524] TIMELINE -> DialRotation
  (bind (_direction _newtrack_0) (Timeline|TimelineDialRotation PlayFromStart)
    (:Update
      ;;[0.538 1 0.5336] Update rotation and check audio cue
      (Transformation|SetRelativeRotation (Variables|Components|GetRotaryPhone_Dial) (Math|Rotator|Lerp(Rotator) (Variables|DialFunctionality|GetDialStartingRotation) (Math|Rotator|MakeRotator 0.0 0.0 _returnvalue_6) _newtrack_0) false false)
      (CallFunction|CheckDialRotation))
    (:Finished
      (Variables|DialFunctionality|SetAudioIndex 1)
      (Components|Timeline|SetPlayRate (Variables|Prop_Quest_BigPhone_Parent|GetTimelineReturnRotation) _returnvalue_4)
      ;;[1 0.884 0.2825] Fade out then fade in return loop
      (Audio|Components|Audio|FadeOut (Variables|Default|GetRotary_InitialSpin_Loop) 0.15 0.0 "Linear" :!class /Script/Engine.AudioComponent)
      (Utilities|FlowControl|Delay 0.15)
      (Components|Activation|Activate _rotary_returnspin_loop false)
      (Audio|Components|Audio|FadeIn _rotary_returnspin_loop 0.1 1.0 0.0 "Linear" :!class /Script/Engine.AudioComponent)
      (bind (_direction_2 _newtrack_0_2) (Timeline|TimelineReturnRotation ReverseFromEnd)
        (:Update
          (Transformation|SetRelativeRotation (Variables|Components|GetRotaryPhone_Dial) (Math|Rotator|Lerp(Rotator) (Variables|DialFunctionality|GetDialStartingRotation) (Math|Rotator|MakeRotator 0.0 0.0 _returnvalue_6) _newtrack_0_2) false false))
        (:Finished
          ;;[1 0.9505 0.1506] Athena_Quest_Rotary_ReturnSpin_Loop_Cue
          ;;
          (Audio|Components|Audio|FadeOut (Variables|Default|GetRotary_ReturnSpin_Loop) 0.1 0.0 "Linear" :!class /Script/Engine.AudioComponent)
          (Audio|Components|Audio|Play (Variables|Default|GetRotary_ReturnSpin_End) 0.0)
          ;;[1 0.1087 0.7842] -> HandleQuestProgression
          (CallFunction|HandleQuestProgression))))))

(event Custom|Check%20Dial%20Rotation
  (bind (_relativerotation_roll _relativerotation_pitch _relativerotation_yaw) (Math|Rotator|BreakRotator (Variables|/Script/Engine.SceneComponent|GetRelativeRotation_Roll (Variables|Components|GetRotaryPhone_Dial))))
  ;[1 0.864 0.04688] Audio hook each time the dial passes a 30 degree increment
  (if (>= (select (< _relativerotation_yaw 0.0) (+ _relativerotation_yaw 3.6e+02) _relativerotation_yaw) (* (Variables|DialFunctionality|GetAudioIndex) (/ 3.6e+02 12.0)))
    (Math|Integer|IncrementInt (Variables|DialFunctionality|GetAudioIndex)
      (:%20%20
        (Audio|PlaySoundatLocation "/Figment_S10/Athena/Sounds/Quests/RotaryPhone/Athena_Quest_Rotary_DialEnd_Cue.Athena_Quest_Rotary_DialEnd_Cue" (Transformation|GetWorldLocation (Variables|Components|GetRotaryPhone_Dial)))))))

(event Custom|PlayCorrectOutgoingRing
  ;[1 0.9711 0.2602] OutgoingRing
  (Audio|Components|Audio|Play (Variables|Default|GetRotary_OutgoingRing) 0.0))

(event Custom|HandleQuestProgression
  ;[0.4018 1 0.4524] Add player to contributed players & track if they have not hit it before
  (bind _local_bplayerdidhitalready (CallFunction|TrackPlayer (Variables|DialFunctionality|GetDialPlayerWhoHit)))
  (if _local_bplayerdidhitalready
    _)
  (Utilities|Array|Add (Variables|yQuest-Backend&BP|GetNumbersEntered) (Variables|DialFunctionality|GetDialNumberHit))
  (bind _iscorrectnumberarray (CallFunction|CheckPhoneNumber (Variables|yQuest-Backend&BP|GetNumbersForQuest)))
  (if _iscorrectnumberarray
    (if (>= (Utilities|Array|LastIndex (Variables|yQuest-Backend&BP|GetNumbersEntered)) (Utilities|Array|LastIndex (Variables|yQuest-Backend&BP|GetNumbersForQuest)))
      (Networking|SwitchHasAuthority
        (:Authority
          (Utilities|FlowControl|Sequence
            (:then_0
              ;[1 0.1087 0.7842] -> AwardQuestProgress
              (CallFunction|GrantQuestCredit))
            (:then_1
              (Default|CallOnPlayerDialed)
              (Utilities|ResetProgressionVariables)
              (CallFunction|PlayCorrectOutgoingRing)))))
      (else
        (Variables|DialFunctionality|SetbDialCanBeHit true)))
    (else
      (Networking|SwitchHasAuthority
        (:Authority
          (CallFunction|PlayIncorrectRing))))))

(event Custom|PlayIncorrectRing
  (bind _rotary_incorrectnumber (Variables|Default|GetRotary_IncorrectNumber))
  (Utilities|ResetProgressionVariables)
  ;[1 0.8941 0.2663] ErrorSound
  (if (not (Audio|Components|Audio|IsPlaying _rotary_incorrectnumber))
    (Audio|Components|Audio|Play _rotary_incorrectnumber 0.0)))

(fn BlueprintModifyIncomingDamage
  (params (Damage :type Float%20%28single-precision%29) (InTags :type Gameplay%20Tag%20Container%20Structure%20%28by%20ref%29) (EffectContext :type Gameplay%20Effect%20Context%20Handle%20Structure%20%28by%20ref%29) (EventInstigator :type Controller%20Object%20Reference) (DamageCauser :type Actor%20Object%20Reference))
  ;[1 1 1] Fix for the % threshold dmg material change on the fort base material that resets in it reverting to its parent actors material
  (return 0.0))

(fn CheckPhoneNumber
  (params (NumberArrayToCheck :type Array%20of%20Integers))
  ;[0.6079 1 0.4362] QuestCheck
  (bind (_array_element _array_index) (Utilities|Array|ForEachLoopwithBreak (Variables|yQuest-Backend&BP|GetNumbersEntered))
    (:LoopBody
      (if (!= _array_element (Utilities|Array|Get(acopy) NumberArrayToCheck _array_index))
        (LocalVariables|SetLocalWrongNumber true)))
    (:Completed
      _))
  (return (not (LocalVariables|GetLocalWrongNumber)) false))

(fn DisplayNumber
  (bind _dialnumberhit (Variables|DialFunctionality|GetDialNumberHit))
  (bind _returnvalue (Utilities|Text|ToText(Integer) _dialnumberhit))
  (switch int (Variables|yQuest-Backend&BP|GetCurrentNumberIndex)
    (:1
      (Rendering|Components|TextRender|SetText (Variables|Default|GetNumber1) _returnvalue))
    (:2
      (Rendering|Components|TextRender|SetText (Variables|Default|GetNumber2) _returnvalue))
    (:3
      (Rendering|Components|TextRender|SetText (Variables|Default|GetNumber3) _returnvalue))
    (:4
      (Rendering|Components|TextRender|SetText (Variables|Default|GetNumber4) _returnvalue))
    (:5
      (for _array_element (Utilities|Array|MakeArray (Variables|Default|GetNumber1) (Variables|Default|GetNumber2) (Variables|Default|GetNumber3) (Variables|Default|GetNumber4))
        (Rendering|Components|TextRender|SetText _array_element (Utilities|Text|MakeLiteralText "")))))
  (return))

(fn GrantQuestCredit
  (Development|PrintString "AwardQuestProgress")
  ;[0.5937 1 0.5123] Award Quest Progress
  (Networking|SwitchHasAuthority
    (:Authority
      (Class|AthenaQuestFunctionLibrary|AthenaGiveQuestUpdatetoPlayers (Variables|yQuest-Backend&BP|GetQuestItemDefinition) (Variables|yQuest-Backend&BP|GetQuestBackendObjectiveName) (Variables|yQuest-Backend&BP|GetObjectiveStat) (Variables|yQuest-Backend&BP|GetTrackedPlayers))
      (Development|PrintString "QuestUpdated" (Variables|zDebug|GetbPrintDebug) (Variables|zDebug|GetbPrintDebug)))))

(fn OnRep_bIsQuestLocallyValid)

(fn PlaySuccessFX
  (params (InController :type Fort%20Player%20Controller%20Object%20Reference))
  (bind _returnvalue (Fortnite|GetLocalFortPlayerControllers true))
  (bind _output (Utilities|Array|Get(acopy) _returnvalue))
  (bind _returnvalue_2 (Game|Player|GetPlayerPawn _output))
  (bind _returnvalue_3 (Transformation|GetActorLocation _returnvalue_2))
  ;[1 0.939 0.5021] Quest Success FX
  (Utilities|IsValid InController
    (:Is%20Valid
      (Development|PrintString "SUCCESS FX" (Variables|zDebug|GetbPrintDebug) (Variables|zDebug|GetbPrintDebug) "(R=1.000000,G=0.703857,B=0.040648,A=1.000000)" 5.0)
      (Effects|SpawnEmitterAttached :EmitterTemplate (Variables|yQuest-Backend&BP|GetSuccessFX) :AttachToComponent (Variables|/Script/Engine.Character|GetMesh _returnvalue_2) :AttachPointName "None" :Location _returnvalue_3 :Scale "1.000000,1.000000,1.000000" :LocationType "SnapToTarget" :bAutoDestroy true :PoolingMethod "None" :bAutoActivate true)
      (Audio|PlaySoundatLocation (Variables|yQuest-Backend&BP|GetSuccessSound) _returnvalue_3) ; Quest Success Sound
    )
    (:Is%20Not%20Valid
      (Development|PrintString "Player controller not valid" (Variables|zDebug|GetbPrintDebug) (Variables|zDebug|GetbPrintDebug)))))

(fn ShouldDie
  (params (Damage :type Float%20%28single-precision%29) (EventInstigator :type Controller%20Object%20Reference) (DamageCauser :type Actor%20Object%20Reference) (bIsTest :type Boolean) (EffectContext :type Gameplay%20Effect%20Context%20Handle%20Structure%20%28by%20ref%29))
  (return false))

(fn TrackPlayer
  (params (Local_PlayerWhoHit :type Fort%20Player%20Controller%20Object%20Reference))
  (bind _localplayerfound (LocalVariables|GetLocalPlayerFound))
  (bind _output_get (LocalVariables|SetLocalPlayer Local_PlayerWhoHit))
  ;[1 1 1] If the quest progress is 0, of course he hasn't been added
  (if (== (Variables|yQuest-Backend&BP|GetCurrentNumberIndex) 0)
    (Utilities|Array|Add (Variables|yQuest-Backend&BP|GetTrackedPlayers) _output_get)
    (return false)
    (else
      (bind _array_element (Utilities|Array|ForEachLoopwithBreak (Variables|yQuest-Backend&BP|GetTrackedPlayers))
        (:LoopBody
          ;[1 1 1] If the player is found, set index found and break loop
          (bind _output_get_2 (LocalVariables|SetLocalPlayerFound (Utilities|Equal(Object) (LocalVariables|GetLocalPlayer) _array_element)))
          (if _output_get_2
            _))
        (:Completed
          _))
      ;[1 1 1] Player was found, give the goods
      (if _localplayerfound
        (return _localplayerfound)
        (else
          (Utilities|Array|Add (Variables|yQuest-Backend&BP|GetTrackedPlayers) (LocalVariables|GetLocalPlayer))
          (return false))))))

(fn ConstructionScript
  (if (Variables|zDebug|GetbSetQuickTestingArray)
    (Utilities|FlowControl|Sequence
      (:then_0
        (Variables|yQuest-Backend&BP|SetNumbersForQuest (Utilities|Array|MakeArray 1)))
      (:then_1
        (Variables|yQuest-Backend&BP|SetNumbersForEasterEgg (Utilities|Array|MakeArray 3)))
      (:then_2
        _))
    (else
      (Utilities|FlowControl|Sequence
        (:then_0
          (Variables|yQuest-Backend&BP|SetNumbersForQuest (Variables|yQuest-Backend&BP|GetNumbersForQuest)))
        (:then_1
          (Variables|yQuest-Backend&BP|SetNumbersForEasterEgg (Variables|yQuest-Backend&BP|GetNumbersForEasterEgg)))
        (:then_2
          _))))
  (Utilities|FlowControl|Sequence
    (:then_0
      (for (_array_element _array_index) (Utilities|Array|MakeArray (Variables|Default|GetNumCollision_0) (Variables|Default|GetNumCollision_1) (Variables|Default|GetNumCollision_2) (Variables|Default|GetNumCollision_3) (Variables|Default|GetNumCollision_4) (Variables|Default|GetNumCollision_5) (Variables|Default|GetNumCollision_6) (Variables|Default|GetNumCollision_7) (Variables|Default|GetNumCollision_8) (Variables|Default|GetNumCollision_9))
        ;[1 1 1] number hit = index
        (Utilities|Map|Add (Variables|DialFunctionality|GetDialCollisionSphereArray) _array_element _array_index)))
    (:then_1
      ;[1 1 1] Lazy color change
      (for _array_element_2 (Utilities|Array|MakeArray "" "" "" "" "" "" "" "" "" "")
        (Rendering|Components|TextRender|SetTextRenderColor _array_element_2 (Utilities|Struct|MakeColor 0))))))
